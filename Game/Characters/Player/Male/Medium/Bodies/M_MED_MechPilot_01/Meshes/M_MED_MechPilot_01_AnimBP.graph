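# Animation Blueprint for the medium male MechPilot_01 body that follows its attached parent mesh and adds corrective offsets to the upper-arm twist bones.
# Each update stores each upper arm's pitch relative to its clavicle, measured as a delta from the reference pose.
# Those angles, range-mapped per side, drive the alpha of additive bone-space Transform Bone nodes applied after Copy Pose From Mesh.
# The shoulderIn_R logic and RunningAround player are unconnected.
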AnimBlueprint /Game/Characters/Player/Male/Medium/Bodies/M_MED_MechPilot_01/Meshes/M_MED_MechPilot_01_AnimBP
target_skeleton=&/M_MED_MechPilot_01_Skeleton
graphs [
 AnimGraph {nodes [
   n1 = "Transform (Modify) Bone - Bone: upperarm_twist_01_l"(
    ComponentPose=Local To Component(LocalPose=Copy Pose From Mesh() {bUseAttachedParent=true})
    Alpha=Get armUp_L() -> (armUp_L:double)
    Translation="-3,0,-5"
    Rotation="0,0,0"
    Scale="1,1,1"
   ) {
    rotationMode=BMM_Additive
    rotationSpace=BCS_BoneSpace
    translationMode=BMM_Additive
    translationSpace=BCS_BoneSpace
    alphaBoolBlend {blendOption=Linear}
    alphaScaleBiasClamp {bMapRange=true, inRange {max=35}}
    boneToModify {boneName=upperarm_twist_01_l}
    translation {x=-3, z=-5}
   }
   n2 = "Transform (Modify) Bone - Bone: upperarm_twist_01_r"(ComponentPose=@n1, Alpha=Get armUp_R() -> (armUp_R:double), Translation="2,0,5", Rotation="0,0,0", Scale="1,1,1") {
    rotationMode=BMM_Additive
    rotationSpace=BCS_BoneSpace
    scaleMode=BMM_Additive
    scaleSpace=BCS_BoneSpace
    translationMode=BMM_Additive
    translationSpace=BCS_BoneSpace
    alphaBoolBlend {blendOption=Linear}
    alphaScaleBiasClamp {bMapRange=true, inRange {max=15}}
    boneToModify {boneName=upperarm_twist_01_r}
    translation {x=2, z=5}
   }
   Output Pose(Result=Component To Local(ComponentPose=@n2)) {layerGroup=None}
   Sequence Player 'RunningAround'() {playRate=0, sequence=&/RunningAround, startPosition=3.48}
   Get armUp_R()
   Get shoulderIn_R() -> (shoulderIn_R:double)
  ]}
 EventGraph {nodes [
   Event BlueprintUpdateAnimation()
   > n1 = Sequence!(execute) -> (then_0, then_1, then_2:Exec, then_3:Exec, then_4:Exec, then_5:Exec)
   n2 = GetOwnerSkeletalMeshComponent()
   n3 = GetDeltaTransformFromRefPose(self=@n2, BoneName=upperarm_l, BaseName=clavicle_l)
   > Set armUp_L(armUp_L:double=BreakRotator(InRot=BreakTransform(InTransform=@n3).Rotation).Pitch) -> (Output_Get:double)
   n4 = GetDeltaTransformFromRefPose(self=@n2, BoneName=upperarm_r, BaseName=clavicle_r)
   Set armUp_R(execute=@n1.then_1, armUp_R:double=BreakRotator(InRot=BreakTransform(InTransform=@n4).Rotation).Pitch) -> (Output_Get:double)
   TryGetPawnOwner()
   BreakRotator(InRot=BreakTransform(InTransform=GetDeltaTransformFromRefPose(self=@n2, BoneName=clavicle_r, BaseName=spine_05)).Rotation)
  ]}
]
variables [
 armUp_L {type=number}
 armUp_R {type=number}
 shoulderIn_R {type=number}
]
